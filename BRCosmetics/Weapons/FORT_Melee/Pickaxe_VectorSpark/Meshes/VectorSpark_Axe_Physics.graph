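# Physics asset for the VectorSpark two-handed pickaxe mesh, defining simulated dangling rope and brush pieces on the weapon.
# Capsule bodies cover the handle, five rope bones and two brush bones.
# Constraints chain the rope from the handle, with tighter swing limits on the first links and looser ones further down.
# Both brushes branch from the last rope link, and every joint has zero twist limit.
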PhysicsAsset /BRCosmetics/Weapons/FORT_Melee/Pickaxe_VectorSpark/Meshes/VectorSpark_Axe_Physics
bodies {
 Melee_TwoHanded_Handle [{shapeType=Capsule, length=55.7489, radius=5.4179, center=<0.0138 1.1227 22.8147>, rotation=<rot 3.5537 -89.9993 0.0244>}]
 dyn_brush_1 [{
   shapeType=Capsule
   length=1.245
   radius=1.0042
   shapeName=dyn_brush_1_capsule
   center=<3.6158 -0.076 -0.2177>
   rotation=<rot 0 89.852 -90.3202>
  }]
 dyn_brush_2 [{
   shapeType=Capsule
   length=1.5591
   radius=0.9037
   shapeName=dyn_brush_2_capsule
   center=<4.1722 -0.4823 -0.1381>
   rotation=<rot 0 -86.0334 -82.371>
  }]
 dyn_rope_1 [{shapeType=Capsule, length=2.81, radius=1.62, shapeName=dyn_rope_1_capsule, center=<2.844 -0.0194 -0.0161>, rotation=<rot 0 179.6101 -90>}]
 dyn_rope_2 [{shapeType=Capsule, length=2.0027, radius=1.6256, shapeName=dyn_rope_2_capsule, center=<2.0848 -0.0678 -0.0161>, rotation=<rot 0 -0.3877 -90>}]
 dyn_rope_3 [{
   shapeType=Capsule
   length=1.5771
   radius=0.6218
   shapeName=dyn_rope_3_capsule
   center=<0.7112 -0.0227 -0.0099>
   rotation=<rot 0 -90.0335 -87.8315>
  }]
 dyn_rope_4 [{
   shapeType=Capsule
   length=1.2991
   radius=0.6309
   shapeName=dyn_rope_4_capsule
   center=<0.5188 -0.0494 -0.0118>
   rotation=<rot 0 -90.4169 -89.096>
  }]
 dyn_rope_5 [{
   shapeType=Capsule
   length=2.2137
   radius=0.5601
   shapeName=dyn_rope_5_capsule
   center=<0.1726 -0.0226 0.0422>
   rotation=<rot -0.0555 90.7498 -89.92>
  }]
}
constraints [
 {bone1Name=dyn_rope_1, bone2Name=Melee_TwoHanded_Handle, swing1LimitDegrees=10, swing2LimitDegrees=80, twistLimitDegrees=0}
 {bone1Name=dyn_rope_2, bone2Name=dyn_rope_1, swing1LimitDegrees=35, swing2LimitDegrees=55, twistLimitDegrees=0}
 {bone1Name=dyn_rope_3, bone2Name=dyn_rope_2, swing1LimitDegrees=80, swing2LimitDegrees=80, twistLimitDegrees=0}
 {bone1Name=dyn_rope_4, bone2Name=dyn_rope_3, swing1LimitDegrees=80, swing2LimitDegrees=80, twistLimitDegrees=0}
 {bone1Name=dyn_rope_5, bone2Name=dyn_rope_4, swing1LimitDegrees=80, swing2LimitDegrees=80, twistLimitDegrees=0}
 {bone1Name=dyn_brush_1, bone2Name=dyn_rope_5, swing1LimitDegrees=50, swing2LimitDegrees=50, twistLimitDegrees=0}
 {bone1Name=dyn_brush_2, bone2Name=dyn_rope_5, swing1LimitDegrees=50, swing2LimitDegrees=50, twistLimitDegrees=0}
]
